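; Spawner actor for Paprika weapon case containers, passing its configured container setting to each container it spawns.
; On server spawn it waits for playlist data to be ready, casts the spawned actor to PaprikaWeaponCase_Container and sets that actor's AV_Container from its own variable.
; BeginPlay calls the parent and then destroys the GmodeMesh component.
; Overlap and Tick events are empty.
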(event EventServerActorSpawned
  (params (Actor :type Actor%20Object%20Reference))
  (Class|FortAsyncActionCurrentPlaylistReady|PlaylistDataReadyAsync
    (:OnReady
      (bind _aspaprika_weapon_case_container (Utilities|Casting|CastToPaprikaWeaponCase_Container Actor))
      (Class|PaprikaWeaponCaseContainer|SetAVContainer _aspaprika_weapon_case_container (Variables|Default|GetAV_Container) :!class /PaprikaWeaponCases/Gameplay/Actors/PaprikaWeaponCase_Container.PaprikaWeaponCase_Container_C))))

(event EventBeginPlay
  (|CallParentFunction)
  (Components|DestroyComponent (Variables|Default|GetGmodeMesh)))

(event Collision|EventActorBeginOverlap
  (params (OtherActor :type Actor%20Object%20Reference)))

(event EventTick
  (params (DeltaSeconds :type Float%20%28single-precision%29)))

(fn ConstructionScript
  (|CallParentFunction))
