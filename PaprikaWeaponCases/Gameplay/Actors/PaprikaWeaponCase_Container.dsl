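; Paprika weapon case, a lootable container that displays a weapon with a mod set and drops it when searched or destroyed.
; The server rolls the spawn chance, waits for loot tables, then picks the weapon and mod set.
; RepNotifies drive client mesh lookups from maps that are then cleared, plus looted cosmetics.
; It also links to its supporting actor, registers with a cluster audio/AV actor and exposes cheat events.
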(event Custom|ForceUpdateRandomModSet
  ;[0.3251 0 1] Cheat ke* commands
  (Networking|SwitchHasAuthority
    (:Authority
      (CallFunction|CheatForceUpdateModsFunc (Variables|ModSetData|GetWeponModSetData)))))

(event Custom|ForceUpdateCovertOpsModSet
  ;[0.3251 0 1] Cheat ke* commands
  (Variables|ModSetData|SetWeponModSetData "/WeaponMods/Items/ModSets/WMSet_CovertOps.WMSet_CovertOps")
  (Networking|SwitchHasAuthority
    (:Authority
      (CallFunction|CheatForceUpdateModsFunc (Variables|ModSetData|GetWeponModSetData)))))

(event Custom|ForceUpdateCoverFireModSet
  ;[0.3251 0 1] Cheat ke* commands
  (Variables|ModSetData|SetWeponModSetData "/WeaponMods/Items/ModSets/WMSet_CoverFire.WMSet_CoverFire")
  (Networking|SwitchHasAuthority
    (:Authority
      (CallFunction|CheatForceUpdateModsFunc (Variables|ModSetData|GetWeponModSetData)))))

(event Custom|ForceUpdateEagleEyeModSet
  ;[0.3251 0 1] Cheat ke* commands
  (Variables|ModSetData|SetWeponModSetData "/WeaponMods/Items/ModSets/WMSet_EagleEye.WMSet_EagleEye")
  (Networking|SwitchHasAuthority
    (:Authority
      (CallFunction|CheatForceUpdateModsFunc (Variables|ModSetData|GetWeponModSetData)))))

(event Custom|ForceUpdateAllRounderModSet
  ;[0.3251 0 1] Cheat ke* commands
  (Variables|ModSetData|SetWeponModSetData "/WeaponMods/Items/ModSets/WMSet_AllRounder.WMSet_AllRounder")
  (Networking|SwitchHasAuthority
    (:Authority
      (CallFunction|CheatForceUpdateModsFunc (Variables|ModSetData|GetWeponModSetData)))))

(event Custom|RemoveSelfFromAVArray
  ;[0 0.05893 1] When a case is opened, remove it from the AV's stored array
  (Class|PaprikaWeaponCaseContainerAudioandAV|RemoveCasefromArrayandDestroySelfonEmptyArray (Variables|AVContainer|GetAV_Container) self))

(event Custom|ForceUpdateRunNGunModSet
  ;[0.3251 0 1] Cheat ke* commands
  (Variables|ModSetData|SetWeponModSetData "/WeaponMods/Items/ModSets/WMSet_RunAndGun.WMSet_RunAndGun")
  (Networking|SwitchHasAuthority
    (:Authority
      (CallFunction|CheatForceUpdateModsFunc (Variables|ModSetData|GetWeponModSetData)))))

(event Custom|ForceUpdateAccuracyModSet
  ;[0.3251 0 1] Cheat ke* commands
  (Variables|ModSetData|SetWeponModSetData "/WeaponMods/Items/ModSets/WMSet_Accuracy.WMSet_Accuracy")
  (Networking|SwitchHasAuthority
    (:Authority
      (CallFunction|CheatForceUpdateModsFunc (Variables|ModSetData|GetWeponModSetData)))))

(event Collision|EventActorBeginOverlap
  (params (OtherActor :type Actor%20Object%20Reference)))

(event Custom|Set%20AV%20Container
  (params (AV%20Container :type Paprika%20Weapon%20Case%20Container%20Audio%20And%20AV%20Object%20Reference))
  ;[0 0.1127 1] AV Container is linked when the actors are placed. B_PaprikaWeaponCase_Container_Spawner calls this event to add the actor it spawns to the array so we can track which cases in the cluster are active / opened
  (Variables|AVContainer|SetAV_Container AV%20Container)
  (Class|PaprikaWeaponCaseContainerAudioandAV|AddCasetoArray (Variables|AVContainer|GetAV_Container) self))

(event Custom|ForceUpdateBushMasterModSet
  ;[0.3251 0 1] Cheat ke* commands
  (Variables|ModSetData|SetWeponModSetData "/WeaponMods/Items/ModSets/WMSet_BushMaster.WMSet_BushMaster"))

(event EventTick
  (params (DeltaSeconds :type Float%20%28single-precision%29)))

(event EventBeginPlay
  (bind _hf_populateweaponincasechance (Variables|Default|GetHF_PopulateWeaponInCaseChance))
  (bind _returnvalue (ScalableFloat|GetValueatLevel _hf_populateweaponincasechance))
  ;[1 1 1] Begin Play
  (|CallParentFunction)
  (Class|FortAsyncActionCurrentPlaylistReady|PlaylistDataReadyAsync
    (:OnReady
      (Utilities|FlowControl|Sequence
        (:then_0
          (Networking|SwitchHasAuthority
            (:Authority
              (Variables|Ammo|SetSpawnAmmo (ScalableFloat|GetValueasBool (Variables|Ammo|GetHF_SpawnAmmo)))
              ;;[0.0697 1 0.1324] If spawn chance is < 100%  e.g. If spawn chance value is 0.8 ( 80% ), anything less 0.8 will pass
              (if (== _returnvalue 1.0)
                _
                (elif (<= (Math|Random|RandomFloatinRange 0.0 1.0) _returnvalue)
                  _
                  (else
                    (CallFunction|SetCaseasAlreadySearched))))
              (CallFunction|InitFindingWeaponandModsEvent 0))
            (:Remote
              (Variables|Default|SetInteractUsesWeaponString (ScalableFloat|GetValueasBool (Variables|Default|GetHF_InteractUsesWeaponString))))))
        (:then_1
          (if (Variables|Default|GetUseCustomStaticMesh)
            (Variables|Default|SetCustomStaticMeshScale (Variables|/Script/Engine.SceneComponent|GetRelativeScale3D (Variables|Default|GetCaseWeaponStaticMesh)))))))))

(event Custom|LoadWeaponMesh
  (bind _weaponsmap (Variables|MeshMaps|GetWeaponsMap))
  ;[0 0.9574 1] Load Weapon Meshes
  (if (Variables|Default|GetUseCustomStaticMesh)
    (Components|DestroyComponent (Variables|Default|GetCaseWeaponStaticMesh))
    (else
      (bind _itemdefinition (Variables|ItemDefs|GetItemDefinition)
        (:then
          (Default|Log "LoadMesh" "Item Definition is valid")
          (bind _itemdata (Class|ItemDefinitionBase|GetItemData _itemdefinition)
            (:Found
              (bind _output_get (Variables|Default|SetUsingSkeletal (Utilities|IsValidSoftObjectReference (Utilities|Struct|BreakFortnitePickupItemData _itemdata))))
              (if _output_get
                ;;[0 0.05186 1] Kill the skel mesh, grab the matching SM and set it
                (Components|DestroyComponent (Variables|Default|GetCaseWeaponSkeletalMesh))
                (if (.ReturnValue (Utilities|Map|Find _weaponsmap (Utilities|Struct|BreakFortnitePickupItemData _itemdata)))
                  (Components|StaticMesh|SetStaticMesh (Variables|Default|GetCaseWeaponStaticMesh) (Utilities|Map|Find _weaponsmap (Utilities|Struct|BreakFortnitePickupItemData _itemdata)))
                  ;;[1 0 0.75] We no longer need the map & refs, clear it out
                  (Utilities|Map|Clear (Variables|MeshMaps|GetWeaponsMap))
                  (else
                    (Development|PrintString (Utilities|String|Append (Utilities|String|ToString(Text) (ItemDefinition|GetDisplayName _itemdefinition)) " WEAPON MESH NOT FOUND"))))))))
        (:else
          (Default|Log "LoadMesh" "Failed, Item Definition == nullptr."))))))

(event Custom|Spawn%20Loot%20%28Server%20Only%29
  (bind _returnvalue (Transformation|GetActorRightVector))
  (bind _returnvalue_2 (Transformation|GetActorLocation))
  (bind _tossrange (Variables|Loot|GetTossRange))
  (bind _returnvalue_3 (* _returnvalue _tossrange))
  (bind _returnvalue_4 (+ _returnvalue_2 _returnvalue_3))
  (bind _weponmodsetdata (Variables|ModSetData|GetWeponModSetData))
  ;[0.2314 0.1848 1] Spawn Loot (Server Only)
  (Networking|SwitchHasAuthority
    (:Authority
      (bind _itempickupsspawned (Class|AthenaFunctionLibrary|SpawnLootfromArray (Variables|Loot|GetLoot) _returnvalue_4 _returnvalue_4 0 true false false 0 "Other" "Chest"))
      (Utilities|IsValid _weponmodsetdata
        (:Is%20Valid
          (for _array_element _itempickupsspawned
            (Class|FortWeaponModFunctionLibrary|ApplyModSettoPickup _weponmodsetdata _array_element)
            (if (Variables|Ammo|GetSpawnAmmo) :!label _label
              (bind _asfort_weapon_item_definition (Utilities|Casting|CastToFortWeaponItemDefinition (Variables|ItemDefs|GetItemDefinition)))
              (Fortnite|Weapon|CreateTossAmmoPickupforWeaponItemDefinitionatLocation _asfort_weapon_item_definition (GameplayTags|MakeLiteralGameplayTagContainer (Variables|Ammo|GetDisplayCaseAmmoTag)) _returnvalue_4))))
        (:Is%20Not%20Valid
          _))
      (goto _label))))

(event Custom|LoadModMeshes
  (bind _weponmodsetdata (Variables|ModSetData|GetWeponModSetData))
  (bind _underbarrelmodsmap (Variables|MeshMaps|GetUnderBarrelModsMap))
  (bind _barrelmodsmap (Variables|MeshMaps|GetBarrelModsMap))
  (bind _opticmodsmap (Variables|MeshMaps|GetOpticModsMap))
  (bind _magazinemodsmap (Variables|MeshMaps|GetMagazineModsMap))
  ;[0 0.9574 1] Load Mod Meshes
  (Default|Log "LoadModMesh" "Spawning Any Mod Mesh for Testing")
  (bind _asfort_weapon_item_definition (Utilities|Casting|CastToFortWeaponItemDefinition (Variables|ItemDefs|GetItemDefinition)))
  (Utilities|IsValid _weponmodsetdata
    (:Is%20Valid
      (bind _returnvalue (Class|FortWeaponModFunctionLibrary|GetModDisplayData _asfort_weapon_item_definition _weponmodsetdata))
      (for _array_element _returnvalue
        (switch Utilities|FlowControl|Switch|SwitchonGameplayTag (Utilities|Struct|BreakFortModDisplayData _array_element)
          (:WeaponMod.Slot.Underbarrel
            (if (.ReturnValue (Utilities|Map|Find _underbarrelmodsmap (Utilities|ResolveSoftReference (.ModMesh (Utilities|Struct|BreakFortModDisplayData _array_element)))))
              (Components|StaticMesh|SetStaticMesh (Variables|Default|GetCaseModPosition3) (Utilities|Map|Find _underbarrelmodsmap (Utilities|ResolveSoftReference (.ModMesh (Utilities|Struct|BreakFortModDisplayData _array_element)))))
              (else
                (Development|PrintString (Utilities|String|Append (Utilities|String|ToString(Text) (ItemDefinition|GetDisplayName (Variables|ItemDefs|GetItemDefinition))) " UNDERBARREL MESH NOT FOUND"))))
            (Utilities|Map|Clear (Variables|MeshMaps|GetUnderBarrelModsMap)))
          (:WeaponMod.Slot.Barrel
            (if (.ReturnValue (Utilities|Map|Find _barrelmodsmap (Utilities|ResolveSoftReference (.ModMesh (Utilities|Struct|BreakFortModDisplayData _array_element)))))
              (Components|StaticMesh|SetStaticMesh (Variables|Default|GetCaseModPosition2) (Utilities|Map|Find _barrelmodsmap (Utilities|ResolveSoftReference (.ModMesh (Utilities|Struct|BreakFortModDisplayData _array_element)))))
              (else
                (Development|PrintString (Utilities|String|Append (Utilities|String|ToString(Text) (ItemDefinition|GetDisplayName (Variables|ItemDefs|GetItemDefinition))) " BARREL MESH NOT FOUND"))))
            (Utilities|Map|Clear (Variables|MeshMaps|GetBarrelModsMap)))
          (:WeaponMod.Slot.Optic
            (if (.ReturnValue (Utilities|Map|Find _opticmodsmap (Utilities|ResolveSoftReference (.ModMesh (Utilities|Struct|BreakFortModDisplayData _array_element)))))
              (Components|StaticMesh|SetStaticMesh (Variables|Default|GetCaseModPosition1) (Utilities|Map|Find _opticmodsmap (Utilities|ResolveSoftReference (.ModMesh (Utilities|Struct|BreakFortModDisplayData _array_element)))))
              (else
                (Development|PrintString (Utilities|String|Append (Utilities|String|ToString(Text) (ItemDefinition|GetDisplayName (Variables|ItemDefs|GetItemDefinition))) " OPTIC MESH NOT FOUND"))))
            (Utilities|Map|Clear (Variables|MeshMaps|GetOpticModsMap)))
          (:WeaponMod.Slot.Magazine
            (if (.ReturnValue (Utilities|Map|Find _magazinemodsmap (Utilities|ResolveSoftReference (.ModMesh (Utilities|Struct|BreakFortModDisplayData _array_element)))))
              (Components|StaticMesh|SetStaticMesh (Variables|Default|GetCaseModPosition4) (Utilities|Map|Find _magazinemodsmap (Utilities|ResolveSoftReference (.ModMesh (Utilities|Struct|BreakFortModDisplayData _array_element)))))
              (else
                (Development|PrintString (Utilities|String|Append (Utilities|String|ToString(Text) (ItemDefinition|GetDisplayName (Variables|ItemDefs|GetItemDefinition))) " MAG MESH NOT FOUND"))))
            (Utilities|Map|Clear (Variables|MeshMaps|GetMagazineModsMap))))))))

(event Custom|InitFindingWeaponAndModsEvent
  (params (Spawner :type BGA%20Paprika%20Weapon%20Cases%20Manager%20Object%20Reference))
  ;[0 1 0.3871] Init
  ;;[1 1 1] Not using the manager to Init, use this for now
  (Variables|Default|SetSpawnerAsSelf self)
  (Default|Log "Init" "called")
  (Utilities|FlowControl|Sequence
    (:then_0
      ;;[0 0.2924 1] If we're not supported from below, we might be on a wall, try linking to that
      (bind _returnvalue (LinkToActorComponent|PerformLinkingCheck (Variables|Default|GetFortLinkToActor) "Down" (Variables|Default|GetLinkedActor)))
      (if (not _returnvalue)
        (LinkToActorComponent|PerformLinkingCheck (Variables|Default|GetFortLinkToActor) "Backward" (Variables|Default|GetLinkedActor))))
    (:then_1
      (Networking|SwitchHasAuthority
        (:Authority
          (bind _asfort_game_mode_athena (Utilities|Casting|CastToFortGameModeAthena (Game|GetGameMode)))
          (if (Class|FortGameModeAthena|HavePlaylistLootTablesBeenApplied _asfort_game_mode_athena)
            (Default|Log "Init" "Loot is ready, Calling ChooseAndInitializeItemDefinition")
            (CallFunction|ChooseandInitializeItemDefinition)
            (else
              (Default|Log "Init" "Loot is NOT ready, binding to OnPlaylistLootTablesAppliedDelegate")
              (EventDispatchers|BindEventtoOnPlaylistLootTablesAppliedDelegate _asfort_game_mode_athena (EventDispatchers|CreateEvent :!event Custom|ChooseAndInitializeItemDefinition)))))))))

(event Custom|PIELoadInMeshes
  ;[0 1 0.3871] Init
  ;;[1 1 1] Not using the manager to Init, use this for now
  (Variables|Default|SetSpawnerAsSelf self)
  (Default|Log "Init" "called")
  (Utilities|FlowControl|Sequence
    (:then_0
      ;;[0 0.2924 1] If we're not supported from below, we might be on a wall, try linking to that
      (bind _returnvalue (LinkToActorComponent|PerformLinkingCheck (Variables|Default|GetFortLinkToActor) "Down" (Variables|Default|GetLinkedActor)))
      (if (not _returnvalue)
        (LinkToActorComponent|PerformLinkingCheck (Variables|Default|GetFortLinkToActor) "Backward" (Variables|Default|GetLinkedActor))))
    (:then_1
      (Networking|SwitchHasAuthority
        (:Authority
          (bind _asfort_game_mode_athena (Utilities|Casting|CastToFortGameModeAthena (Game|GetGameMode)))
          (if (Class|FortGameModeAthena|HavePlaylistLootTablesBeenApplied _asfort_game_mode_athena)
            (Default|Log "Init" "Loot is ready, Calling ChooseAndInitializeItemDefinition")
            (CallFunction|ChooseandInitializeItemDefinition)
            (else
              (Default|Log "Init" "Loot is NOT ready, binding to OnPlaylistLootTablesAppliedDelegate")
              (EventDispatchers|BindEventtoOnPlaylistLootTablesAppliedDelegate _asfort_game_mode_athena (EventDispatchers|CreateEvent :!event Custom|ChooseAndInitializeItemDefinition)))))))))

(event Custom|ChooseAndInitializeItemDefinition
  ;[0 1 0.5812] ChooseAndInitializeItemDefinition
  (Networking|SwitchHasAuthority
    (:Authority
      (Class|FortAsyncActionAuthorityMatchReady|MatchReadyAsync
        (:OnReady
          (bind (_asfort_weapon_item_definition _loot) (Default|GetWeaponfromLootTiertoDisplay))
          (Default|Log "ChooseAndInitializeItemDefinition" (Utilities|String|Append "Item Definition: [" (Utilities|String|ToString(Text) (ItemDefinition|GetDisplayName _asfort_weapon_item_definition)) "]"))
          ;;[1 1 1] Not using the manager to Init, use this for now
          (Utilities|Array|Add (Variables|Default|GetGeneratedItems) (Weapon|GetUpgradedWeaponItemVerticaltoRarity _asfort_weapon_item_definition "Legendary"))
          (Variables|ItemDefs|SetItemDefinition _asfort_weapon_item_definition)
          (Variables|Loot|SetLoot _loot)
          (Networking|ForceNetUpdate)
          (CallFunction|UnbindLootTablesApplied)
          (Utilities|FlowControl|Delay 1.0)
          ;;[1 0 0.75] We no longer need the map & refs, clear it out
          (Utilities|Map|Clear (Variables|MeshMaps|GetWeaponsMap))
          (Utilities|Map|Clear (Variables|MeshMaps|GetUnderBarrelModsMap))
          (Utilities|Map|Clear (Variables|MeshMaps|GetBarrelModsMap))
          (Utilities|Map|Clear (Variables|MeshMaps|GetOpticModsMap))
          (Utilities|Map|Clear (Variables|MeshMaps|GetMagazineModsMap)))) ; TODO: Remove after correct loot tiers. Trying to async this more with faked grab loot method
    )))

(event Custom|ForceUpdateRandomWeaponAndMods
  ;[0 1 0.5812] ChooseAndInitializeItemDefinition
  (Networking|SwitchHasAuthority
    (:Authority
      (Class|FortAsyncActionAuthorityMatchReady|MatchReadyAsync
        (:OnReady
          (bind (_asfort_weapon_item_definition _loot) (Default|GetWeaponfromLootTiertoDisplay))
          (Default|Log "ChooseAndInitializeItemDefinition" (Utilities|String|Append "Item Definition: [" (Utilities|String|ToString(Text) (ItemDefinition|GetDisplayName _asfort_weapon_item_definition)) "]"))
          ;;[1 1 1] Not using the manager to Init, use this for now
          (Utilities|Array|Add (Variables|Default|GetGeneratedItems) (Weapon|GetUpgradedWeaponItemVerticaltoRarity _asfort_weapon_item_definition "Legendary"))
          (Variables|ItemDefs|SetItemDefinition _asfort_weapon_item_definition)
          (Variables|Loot|SetLoot _loot)
          (Networking|ForceNetUpdate)
          (CallFunction|UnbindLootTablesApplied)
          (Utilities|FlowControl|Delay 1.0)
          ;;[1 0 0.75] We no longer need the map & refs, clear it out
          (Utilities|Map|Clear (Variables|MeshMaps|GetWeaponsMap))
          (Utilities|Map|Clear (Variables|MeshMaps|GetUnderBarrelModsMap))
          (Utilities|Map|Clear (Variables|MeshMaps|GetBarrelModsMap))
          (Utilities|Map|Clear (Variables|MeshMaps|GetOpticModsMap))
          (Utilities|Map|Clear (Variables|MeshMaps|GetMagazineModsMap)))) ; TODO: Remove after correct loot tiers. Trying to async this more with faked grab loot method
    )))

(event Custom|ForceSuperRareWeaponRarity
  ;[0 1 0.5812] ChooseAndInitializeItemDefinition
  (Variables|Debug|SetForceRarityTag "(TagName=\"Rarity.SuperRare\")")
  (Variables|Debug|SetbForceRarity true)
  (Networking|SwitchHasAuthority
    (:Authority
      (Class|FortAsyncActionAuthorityMatchReady|MatchReadyAsync
        (:OnReady
          (bind (_asfort_weapon_item_definition _loot) (Default|GetWeaponfromLootTiertoDisplay))
          (Default|Log "ChooseAndInitializeItemDefinition" (Utilities|String|Append "Item Definition: [" (Utilities|String|ToString(Text) (ItemDefinition|GetDisplayName _asfort_weapon_item_definition)) "]"))
          ;;[1 1 1] Not using the manager to Init, use this for now
          (Utilities|Array|Add (Variables|Default|GetGeneratedItems) (Weapon|GetUpgradedWeaponItemVerticaltoRarity _asfort_weapon_item_definition "Legendary"))
          (Variables|ItemDefs|SetItemDefinition _asfort_weapon_item_definition)
          (Variables|Loot|SetLoot _loot)
          (Networking|ForceNetUpdate)
          (CallFunction|UnbindLootTablesApplied)
          (Utilities|FlowControl|Delay 1.0)
          ;;[1 0 0.75] We no longer need the map & refs, clear it out
          (Utilities|Map|Clear (Variables|MeshMaps|GetWeaponsMap))
          (Utilities|Map|Clear (Variables|MeshMaps|GetUnderBarrelModsMap))
          (Utilities|Map|Clear (Variables|MeshMaps|GetBarrelModsMap))
          (Utilities|Map|Clear (Variables|MeshMaps|GetOpticModsMap))
          (Utilities|Map|Clear (Variables|MeshMaps|GetMagazineModsMap)))) ; TODO: Remove after correct loot tiers. Trying to async this more with faked grab loot method
    )))

(event Custom|ForceVeryRareWeaponRarity
  ;[0 1 0.5812] ChooseAndInitializeItemDefinition
  (Variables|Debug|SetForceRarityTag "(TagName=\"Rarity.VeryRare\")")
  (Variables|Debug|SetbForceRarity true)
  (Networking|SwitchHasAuthority
    (:Authority
      (Class|FortAsyncActionAuthorityMatchReady|MatchReadyAsync
        (:OnReady
          (bind (_asfort_weapon_item_definition _loot) (Default|GetWeaponfromLootTiertoDisplay))
          (Default|Log "ChooseAndInitializeItemDefinition" (Utilities|String|Append "Item Definition: [" (Utilities|String|ToString(Text) (ItemDefinition|GetDisplayName _asfort_weapon_item_definition)) "]"))
          ;;[1 1 1] Not using the manager to Init, use this for now
          (Utilities|Array|Add (Variables|Default|GetGeneratedItems) (Weapon|GetUpgradedWeaponItemVerticaltoRarity _asfort_weapon_item_definition "Legendary"))
          (Variables|ItemDefs|SetItemDefinition _asfort_weapon_item_definition)
          (Variables|Loot|SetLoot _loot)
          (Networking|ForceNetUpdate)
          (CallFunction|UnbindLootTablesApplied)
          (Utilities|FlowControl|Delay 1.0)
          ;;[1 0 0.75] We no longer need the map & refs, clear it out
          (Utilities|Map|Clear (Variables|MeshMaps|GetWeaponsMap))
          (Utilities|Map|Clear (Variables|MeshMaps|GetUnderBarrelModsMap))
          (Utilities|Map|Clear (Variables|MeshMaps|GetBarrelModsMap))
          (Utilities|Map|Clear (Variables|MeshMaps|GetOpticModsMap))
          (Utilities|Map|Clear (Variables|MeshMaps|GetMagazineModsMap)))) ; TODO: Remove after correct loot tiers. Trying to async this more with faked grab loot method
    )))

(event Custom|ForceRareWeaponRarity
  ;[0 1 0.5812] ChooseAndInitializeItemDefinition
  (Variables|Debug|SetForceRarityTag "(TagName=\"Rarity.Rare\")")
  (Variables|Debug|SetbForceRarity true)
  (Networking|SwitchHasAuthority
    (:Authority
      (Class|FortAsyncActionAuthorityMatchReady|MatchReadyAsync
        (:OnReady
          (bind (_asfort_weapon_item_definition _loot) (Default|GetWeaponfromLootTiertoDisplay))
          (Default|Log "ChooseAndInitializeItemDefinition" (Utilities|String|Append "Item Definition: [" (Utilities|String|ToString(Text) (ItemDefinition|GetDisplayName _asfort_weapon_item_definition)) "]"))
          ;;[1 1 1] Not using the manager to Init, use this for now
          (Utilities|Array|Add (Variables|Default|GetGeneratedItems) (Weapon|GetUpgradedWeaponItemVerticaltoRarity _asfort_weapon_item_definition "Legendary"))
          (Variables|ItemDefs|SetItemDefinition _asfort_weapon_item_definition)
          (Variables|Loot|SetLoot _loot)
          (Networking|ForceNetUpdate)
          (CallFunction|UnbindLootTablesApplied)
          (Utilities|FlowControl|Delay 1.0)
          ;;[1 0 0.75] We no longer need the map & refs, clear it out
          (Utilities|Map|Clear (Variables|MeshMaps|GetWeaponsMap))
          (Utilities|Map|Clear (Variables|MeshMaps|GetUnderBarrelModsMap))
          (Utilities|Map|Clear (Variables|MeshMaps|GetBarrelModsMap))
          (Utilities|Map|Clear (Variables|MeshMaps|GetOpticModsMap))
          (Utilities|Map|Clear (Variables|MeshMaps|GetMagazineModsMap)))) ; TODO: Remove after correct loot tiers. Trying to async this more with faked grab loot method
    )))

(event Custom|UnbindLootTablesApplied
  ;[1 0.6091 0] UnbindLootTablesApplied
  (bind _asfort_game_mode_athena (Utilities|Casting|CastToFortGameModeAthena (Game|GetGameMode)))
  (EventDispatchers|UnbindEventfromOnPlaylistLootTablesAppliedDelegate _asfort_game_mode_athena (EventDispatchers|CreateEvent :!event Custom|ChooseAndInitializeItemDefinition)))

(event EventOnSetSearched
  ;[1 0 0.6291] On looted, spawn loot, play cosmetic events & remove from AV container array for audio param updates
  (|CallParentFunction)
  (CallFunction|SpawnLoot(ServerOnly))
  (Variables|Interact|SetDisableInteraction true)
  (Variables|Loot|SetAlreadyLooted true)
  (CallFunction|RemoveSelffromAVArray))

(event EventEndPlay
  (params (EndPlayReason :type EEndPlayReason%20Enum :default "Destroyed"))
  ;[1 0.5326 0] EndPlay
  (|CallParentFunction EndPlayReason)
  (CallFunction|UnbindLootTablesApplied))

(event Custom|OnInteractCosmeticEvents
  ;[0 0.1811 1] Called in AlreadyLooted Rep Notify. Make sure clients see cases in looted cosmetic state
  (Networking|SwitchHasAuthority
    (:Remote
      (GameplayCue|ExecuteGameplayCue 0 (Variables|GameplayCues|GetOn%20Interact%20Gameplay%20Cue%20Tag))
      (Components|Activation|Activate (Variables|Default|GetInteractFlashFX) false)
      (Components|Activation|Deactivate (Variables|Default|GetCustom%20Chest%20Aura))
      (Rendering|Material|SetCustomPrimitiveDataFloat (Variables|Default|GetLightingPlane) 0 1.0)
      (Rendering|Material|SetMaterial (Variables|Debug|GetStaticMeshComponent) 0 "/PaprikaWeaponCases/Assets/Materials/MI_CacheWeapon_Closed.MI_CacheWeapon_Closed"))))

(event OnLinkedActorDestroyed(FortLinkToActor)
  (params (DamageCauser :type Actor%20Object%20Reference))
  (CallFunction|HandleLinkedActorDestroyed))

(event EventOnDeathServer
  (params (Damage :type Float%20%28single-precision%29) (DamageTags :type Gameplay%20Tag%20Container%20Structure%20%28by%20ref%29) (Momentum :type Vector :default "0, 0, 0") (HitInfo :type Hit%20Result%20Structure%20%28by%20ref%29) (InstigatedBy :type Controller%20Object%20Reference) (DamageCauser :type Actor%20Object%20Reference) (EffectContext :type Gameplay%20Effect%20Context%20Handle%20Structure))
  (if (Variables|Loot|GetAlreadyLooted)
    _
    (elif (and (not (Container|IsSpatialGameplayInterfaceDestroyingActors)) (not (GameplayTags|HasAnyTags DamageTags (GameplayTags|MakeLiteralGameplayTagContainer "(GameplayTags=((TagName=\"Gameplay.Effect.InstantDeath.Silent\"),(TagName=\"Gameplay.Effect.InstantDeath.Culled\")))"))))
      (CallFunction|SpawnLoot(ServerOnly))
      (Variables|Interact|SetDisableInteraction true)
      (Variables|Loot|SetAlreadyLooted true)
      (CallFunction|RemoveSelffromAVArray))))

(event EventOnDeathPlayEffects
  (params (Damage :type Float%20%28single-precision%29) (DamageTags :type Gameplay%20Tag%20Container%20Structure%20%28by%20ref%29) (Momentum :type Vector :default "0, 0, 0") (HitInfo :type Hit%20Result%20Structure%20%28by%20ref%29) (InstigatedBy :type Fort%20Pawn%20Object%20Reference) (DamageCauser :type Actor%20Object%20Reference) (EffectContext :type Gameplay%20Effect%20Context%20Handle%20Structure))
  (GameplayCue|ExecuteGameplayCue 0 (Variables|GameplayCues|GetOn%20Destroyed%20Gameplay%20Cue%20Tag)))

(fn BlueprintCanInteract
  (params (InteractingPawn :type Fort%20Pawn%20Object%20Reference) (InteractionBeingAttempted :type EInteractionBeingAttempted%20Enum :default "FirstInteraction") (InteractionType :type TInteractionType%20Enum :default "IT_NoInteraction"))
  (return true))

(fn BlueprintGetAllHighlightableComponents
  (return (Utilities|Array|MakeArray (Variables|Debug|GetStaticMeshComponent)) false))

(fn BlueprintGetFailedInteractionString
  (params (InteractingPawn :type Fort%20Pawn%20Object%20Reference) (InteractionBeingAttempted :type EInteractionBeingAttempted%20Enum :default "FirstInteraction"))
  (return ""))

(fn BlueprintGetInteractionString
  (params (InteractingPawn :type Fort%20Pawn%20Object%20Reference) (InteractionBeingAttempted :type EInteractionBeingAttempted%20Enum :default "FirstInteraction"))
  (if (Variables|Default|GetInteractUsesWeaponString)
    (bind _weponmodsetdata (Variables|ModSetData|GetWeponModSetData)
      (:then
        (bind _itemdefinition (Variables|ItemDefs|GetItemDefinition)
          (:then
            (return (Utilities|Text|FormatText "NSLOCTEXT(\"[4EA3DCEA525C80A7EB256162C303DACF]\", \"ED238D04419B405924D967B2E7731742\", \"{ModSetName}\\r\\n{WeaponName}\")" (Variables|/Script/FortniteGame.FortWeaponModSetData|GetModSetName _weponmodsetdata) (ItemDefinition|GetDisplayName _itemdefinition))))
          (:else
            _)))
      (:else
        _)))
  (return (Variables|Interact|GetInteractString)))

(fn BuildExposedAssetProviderVersion(Implementation)
  (params (Version :type Fort%20Exposed%20Asset%20Provider%20Version%20Structure%20%28by%20ref%29))
  (|CallParentFunction Version)
  (ExposedAssetProvider|AddVersion Version 1)
  (return))

(fn CheatForceUpdateModsFunc
  (params (ModSet :type Fort%20Weapon%20Mod%20Set%20Data%20Object%20Reference))
  ;[0.4816 0 1] Cheat ke* ForceUpdateMods
  (bind _array_element (Utilities|Array|ForEachLoopwithBreak (Variables|Loot|GetLoot))
    (:LoopBody
      (bind _asfort_weapon_item_definition (Utilities|Casting|CastToFortWeaponItemDefinition (Class|FortItemFunctionLibrary|BreakFortItemEntry _array_element)))
      (Utilities|IsValid ModSet
        (:Is%20Valid
          (Variables|ModSetData|SetWeponModSetData ModSet))
        (:Is%20Not%20Valid
          (bind _returnvalue (Class|FortWeaponModFunctionLibrary|GetRandomWeaponModSetData _asfort_weapon_item_definition))
          (Variables|ModSetData|SetWeponModSetData _returnvalue)))
      (Variables|Debug|SetDebugUpdateNotifyInt (+ (Variables|Debug|GetDebugUpdateNotifyInt) 1))
      (Networking|ForceNetUpdate))))

(fn CollectAssetReferencestoAllow(Implementation)
  (params (Collector :type Fort%20Exposed%20Asset%20Reference%20Collector%20Structure%20%28by%20ref%29))
  (|CallParentFunction Collector)
  (ExposedAssetProvider|CollectIndirectAssetReference Collector (LocalVariables|GetM_PaprikaWeaponCase_SideLight))
  (ExposedAssetProvider|CollectIndirectAssetReference Collector (LocalVariables|GetM_PaprikaWeaponCase_Padding_Velvet))
  (return))

(fn GetWeaponFromLootTierToDisplay
  (bind _bforcerarity (Variables|Debug|GetbForceRarity))
  ;[1 1 1] Temp logic for playtesting
  (bind (_outloottodrop _returnvalue) (Loot|PickLootDrops (Variables|Loot|GetLootTierGroupName) 0 -1 "(GameplayTags=)" :!label _label_3))
  (if _returnvalue
    (bind _output_get (LocalVariables|SetCurrLoot _outloottodrop))
    (bind _array_element (Utilities|Array|ForEachLoopwithBreak _output_get :!label _label)
      (:LoopBody
        (bind _asfort_weapon_item_definition (Utilities|Casting|CastToFortWeaponItemDefinition (Class|FortItemFunctionLibrary|BreakFortItemEntry _array_element)))
        (bind _itemdata (Class|ItemDefinitionBase|GetItemData (Class|FortItemFunctionLibrary|BreakFortItemEntry _array_element))
          (:Found
            (if (not (Math|Boolean|ANDBoolean (not (GameplayTags|HasMatchingGameplayTag _asfort_weapon_item_definition "(TagName=\"Item.Ammo\")")) (Utilities|IsValidSoftObjectReference (Utilities|Struct|BreakFortnitePickupItemData _itemdata)) (GameplayTags|HasMatchingGameplayTag _asfort_weapon_item_definition "(TagName=\"ModSetBucket\")") (not _bforcerarity)))
              (if (and _bforcerarity (GameplayTags|HasMatchingGameplayTag _asfort_weapon_item_definition (Variables|Debug|GetForceRarityTag))) :!label _label_2
                _))
            (bind _returnvalue_2 (Class|FortWeaponModFunctionLibrary|GetRandomWeaponModSetData _asfort_weapon_item_definition))
            (Variables|ModSetData|SetWeponModSetData _returnvalue_2)
            (if (Utilities|Array|ContainsItem (Variables|Default|GetGeneratedItems) (Weapon|GetUpgradedWeaponItemVerticaltoRarity _asfort_weapon_item_definition "Legendary"))
              (LocalVariables|SetLastValidWeapon _asfort_weapon_item_definition)
              (goto _label Break)
              (else
                ;;[1 0.5052 0.5052] Found an item that fits our spec, return it.
                (return _asfort_weapon_item_definition (LocalVariables|GetCurrLoot)))))
          (:NotFound
            _))
        (goto _label_2))
      (:Completed
        (bind _result (Math|Integer|IncrementInt (LocalVariables|GetAttemtps))
          (:%20%20
            (if (> _result (Variables|Default|GetTempPickLootLoopAttemps))
              ;;[1 0.5052 0.5052] Failed to find a valid item after a failsafe number of attempts, return the last valid item.
              (return (LocalVariables|GetLastValidWeapon) (LocalVariables|GetCurrLoot))
              (else
                _))))))
    (else
      (bind _result_2 (Math|Integer|IncrementInt (LocalVariables|GetAttemtps))
        (:%20%20
          (if (> _result_2 (Variables|Default|GetTempPickLootLoopAttemps))
            ;;[1 0.5052 0.5052] Failed to find item, return nothing.
            (return 0 (LocalVariables|GetCurrLoot))
            (else
              _))))))
  (goto _label_3))

(fn HandleLinkedActorDestroyed
  (Actor|SetLifeSpan self 0.25))

(fn Log
  (params (FunctionName :type String) (Message :type String))
  (if (Variables|Default|GetbLoggingEnabled%20Const)
    (Fortnite|FortShippingLog self (Utilities|String|ToString(Text) (Utilities|Text|FormatText "INVTEXT(\"PaprikaWeaponCases::{FunctionName}() object: [{ObjectName}] {Message}.\")" FunctionName (Utilities|GetObjectName self) Message)))))

(fn OnRep_AlreadyLooted
  (bind _casemodposition1 (Variables|Default|GetCaseModPosition1))
  (Rendering|SetVisibility _casemodposition1 false false)
  (Components|DestroyComponent _casemodposition1)
  (CallFunction|OnInteractCosmeticEvents))

(fn OnRep_AV_Container
  (Class|PaprikaWeaponCaseContainerAudioandAV|AddCasetoArray (Variables|AVContainer|GetAV_Container) self))

(fn OnRep_DebugUpdateNotifyInt
  (Networking|CanExecuteCosmeticEvents
    (:True
      (CallFunction|LoadModMeshes))))

(fn OnRep_DisableInteraction
  (Variables|Interaction|SetbAllowInteract false)
  (Variables|Interaction|SetbShowFirstInteractPrompt false))

(fn OnRep_ItemDefinition
  (Default|UpdateItemDefinition))

(fn SetCaseAsAlreadySearched
  (Variables|Interact|SetDisableInteraction true)
  (Variables|Loot|SetAlreadyLooted true)
  (CallFunction|RemoveSelffromAVArray))

(fn UpdateItemDefinition
  (Networking|CanExecuteCosmeticEvents
    (:True
      (CallFunction|LoadWeaponMesh)
      (CallFunction|LoadModMeshes))))

(fn ConstructionScript
  (|CallParentFunction))
