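; Non-interactable audio and AV actor shared by a group of Paprika weapon cases, providing their ambient sound.
; The authority adds cases to an array whose OnRep sets the ambient sound's case count and moves it to their average location.
; Removing a case multicasts so remote clients refresh audio and play the open sound, and an emptied array expires the actor shortly after.
; Once playlist data is ready, it destroys its glow and smoke FX and activates the sound indicator.
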(event Custom|MultiCast_UpdateContainerArrayForClientSideAudioParamUpdate
  (params (WeaponCase :type Paprika%20Weapon%20Case%20Container%20Object%20Reference))
  ;[0 0.1243 1] On removal, update container array size to adjust audio param for client audio
  (Networking|SwitchHasAuthority
    (:Remote
      (CallFunction|UpdateAudioParameters)
      (Audio|PlaySoundatLocation (Variables|Default|GetOverrideOpenSoundCue) (Transformation|GetActorLocation WeaponCase)))))

(event Custom|AddCaseToArray
  (params (Weapon%20Case :type Paprika%20Weapon%20Case%20Container%20Object%20Reference))
  (bind _weaponcasecontainerarray (Variables|Default|GetWeaponCaseContainerArray))
  (Utilities|IsValid Weapon%20Case
    (:Is%20Valid
      (Networking|SwitchHasAuthority
        (:Authority
          (Utilities|Array|AddUnique _weaponcasecontainerarray Weapon%20Case)
          (Variables|Default|SetWeaponCaseContainerArray _weaponcasecontainerarray))))))

(event Collision|EventActorBeginOverlap
  (params (OtherActor :type Actor%20Object%20Reference)))

(event Custom|RemoveCaseFromArrayAndDestroySelfOnEmptyArray
  (params (Weapon%20Case :type Paprika%20Weapon%20Case%20Container%20Object%20Reference))
  (bind _weaponcasecontainerarray (Variables|Default|GetWeaponCaseContainerArray))
  ;[0 0.1243 1] Remove the case and if all cases in the array are looted, destroy this AV actor to stop audio & AV
  (Utilities|Array|RemoveItem _weaponcasecontainerarray Weapon%20Case)
  (Utilities|FlowControl|Sequence
    (:then_0
      (CallFunction|MultiCastUpdateContainerArrayforClientSideAudioParamUpdate Weapon%20Case))
    (:then_1
      (Utilities|FlowControl|Delay 0.5)
      (if (== (Utilities|Array|Length _weaponcasecontainerarray) 0)
        (Actor|SetLifeSpan 0.2)))))

(event EventTick
  (params (DeltaSeconds :type Float%20%28single-precision%29)))

(event EventBeginPlay
  ;[1 1 1] Begin Play
  (|CallParentFunction)
  (Class|FortAsyncActionCurrentPlaylistReady|PlaylistDataReadyAsync
    (:OnReady
      (Components|DestroyComponent (Variables|Default|GetFX%20Glow%20&%20Smoke%20Sheet))
      (Components|Activation|Activate (Variables|Default|GetSound%20Indicator%20Component) false))))

(event EventEndPlay
  (params (EndPlayReason :type EEndPlayReason%20Enum :default "Destroyed"))
  ;[1 0.5326 0] EndPlay
  (|CallParentFunction EndPlayReason))

(fn BlueprintCanInteract
  (params (InteractingPawn :type Fort%20Pawn%20Object%20Reference) (InteractionBeingAttempted :type EInteractionBeingAttempted%20Enum :default "FirstInteraction") (InteractionType :type TInteractionType%20Enum :default "IT_NoInteraction"))
  (return false))

(fn BlueprintGetAllHighlightableComponents
  (return (Utilities|Array|MakeArray (Variables|Debug|GetStaticMeshComponent)) false))

(fn BlueprintGetFailedInteractionString
  (params (InteractingPawn :type Fort%20Pawn%20Object%20Reference) (InteractionBeingAttempted :type EInteractionBeingAttempted%20Enum :default "FirstInteraction"))
  (return ""))

(fn BlueprintGetInteractionString
  (params (InteractingPawn :type Fort%20Pawn%20Object%20Reference) (InteractionBeingAttempted :type EInteractionBeingAttempted%20Enum :default "FirstInteraction"))
  (return ""))

(fn OnRep_WeaponCaseContainerArray
  (CallFunction|UpdateAudioParameters))

(fn UpdateAudioParameters
  (bind _weaponcasecontainerarray (Variables|Default|GetWeaponCaseContainerArray))
  (bind _returnvalue (Utilities|Array|Length _weaponcasecontainerarray))
  (Audio|Parameter|SetIntegerParameter (Variables|Default|GetChest_Ambient_Sound) "NumberOfCases" _returnvalue :!class /Script/Engine.AudioComponent)
  (if (> _returnvalue 0)
    (bind _returnvalue_2 (Transformation|GetActorArrayAverageLocation _weaponcasecontainerarray))
    (Transformation|SetWorldLocation (Variables|Default|GetChest_Ambient_Sound) _returnvalue_2 false false)))

(fn ConstructionScript
  (|CallParentFunction))
